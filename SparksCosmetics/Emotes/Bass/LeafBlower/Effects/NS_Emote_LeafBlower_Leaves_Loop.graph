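# Looping Niagara effect of autumn leaves blown upward for the leaf blower bass emote in SparksCosmetics.
# Three sprite emitters continuously spawn leaves from a four-color palette with SubUV animation, drag, random hue/brightness and time-of-day dimming.
# Leaves_01 and Leaves_02 throw short-lived leaves in cones, with curl noise swirling the first.
# Leaves_Collision adds sparse, long-lived leaves that fall under gravity, settle on collision, age faster and freeze their flipbook.
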NiagaraSystem /SparksCosmetics/Emotes/Bass/LeafBlower/Effects/NS_Emote_LeafBlower_Leaves_Loop
dependencies {
 usedDataInterfaces [
  &NiagaraDataInterfaceArrayColor
  &NiagaraDataInterfaceSpriteRendererInfo
  &NiagaraDataInterfaceCurve
 ]
 usedDynamicInputs []
 usedModules [&SystemState]
}
emitters {
 Leaves_01 {
  emitterUpdateScript [
   &EmitterState()
   &SpawnRate(SpawnRate=30, Use Spawn Probability=false)
  ]
  particleSpawnScript [
   &InitializeParticle(
    Color=&SelectLinearColorFromArray(
     Array Sampling Mode=::NewEnumerator0
     Direct Array Index=?
     Color Selection Array=DataInterface() {propertyValues {ColorData [<rgba 0.4427 0.0503 0.0679 1>, <rgba 0.9911 0.6038 0.0999 1>, <rgba 0.3489 0.3763 0.0746 1>, <rgba 1 0.3419 0 1>]}}
     Random Seed=0
    )
    Lifetime Max=0.8
    Lifetime Min=0.4
    Lifetime Mode=::NewEnumerator1
    Mesh Scale Mode=::NewEnumerator3
    Sprite Rotation Mode=::NewEnumerator2
    Sprite Size Max=<15 15>
    Sprite Size Min=<7.5 7.5>
    Sprite Size Mode=::NewEnumerator2
    UsePositionOffset=false
    Write Material Random=true
   )
   &ShapeLocation(
    Hemisphere Distribution=<1 0.4>
    MaskSpawnGroup=false
    Override Seed=false
    Rotation Coordinate Space=::World
    Sphere Radius=5
    UseOwnerScale=true
   )
   &SystemLocation(MaskSpawnGroup=false, Offset=<0 0 40>)
   &AddVelocity(
    Cone Angle=35
    Cone Axis=<0 0.3 1>
    Module.UseConeFalloff=false
    Module.UseVelDistribution=false
    Override Seed=false
    Velocity Mode=::NewEnumerator2
    Velocity Speed=&RandomRangeFloat(Minimum=300, Maximum=600)
   ) {module=AddVelocity001}
  ]
  particleUpdateScript [
   &ParticleState()
   &SubUVAnimation(
    Play Rate:NiagaraFloat=&RandomRangeFloat(Minimum=1, Maximum=3)
    Playback Scale Curve=DataInterface() {propertyValues {Curve {keys [{interpMode=RCIM_Cubic, tangentMode=RCTM_User, time=0, value=0}, {interpMode=RCIM_Cubic, tangentMode=RCTM_User, time=1, value=1}]}}}
    "Sprite Renderer "=DataInterface() {propertyValues {SpriteRenderer=&self:Leaves_01_3.NiagaraSpriteRendererProperties_1}}
    SubUV Animation Mode:ENiagara_SubUVLookupModeV2=::NewEnumerator3
    UseEndFrame=true
    UsePlayRateScale=true
    UseStartFrame=true
   ) {module=SubUVAnimation002}
   &AccelerationForce(Acceleration=&RandomRangeVector(Minimum=<0 0 -50>, Maximum=<0 0 -100>))
   &Drag(
    Drag=&FloatFromCurve(
     FloatCurve=DataInterface() {propertyValues {Curve {keys [
         {arriveTangent=3.7695, interpMode=RCIM_Cubic, leaveTangent=3.7695, tangentMode=RCTM_User, time=0, value=0}
         {interpMode=RCIM_Cubic, tangentMode=RCTM_User, time=1, value=1}
        ]}}}
     Scale Curve=2
    )
    Use Linear Drag=true
    Use Rotational Drag=true
   )
   &DynamicMaterialParameters(
    Index 0 Param 1=0.1
    Module.Param0WriteEnabled=true
    Module.Param1WriteEnabled=true
    Module.Param2WriteEnabled=true
    Module.Param3WriteEnabled=true
   )
   &ScaleColor(
    Scale Alpha=&FloatFromCurve(FloatCurve=DataInterface() {propertyValues {Curve {keys [
         {interpMode=RCIM_Cubic, tangentMode=RCTM_User, time=0, value=0}
         {arriveTangent=-0.0948, interpMode=RCIM_Cubic, leaveTangent=-0.0948, tangentMode=RCTM_User, time=0.1, value=1}
         <at 0.942 0.9825>
         <at 1 0>
        ]}}})
    Scale RGB=&MakeVectorFromLinearColorRGB(Color=&HueShiftLinearColor(
      Hue=&RandomRangeFloat(Minimum=0, Maximum=0.05)
      Saturation=1
      Value=&Multiply_Float(
       A=&RandomRangeFloat(Minimum=1.3, Maximum=4)
       B=&FNTimeOfDayScalar(Night Scale=0.5, Morning Scale=1, Evening Scale=0.8)
      )
      Linear Color=$Particles.Initial.Color
     ))
    ScaleA=false
    ScaleRGB=true
   )
   &ScaleSpriteSize(Uniform Curve Sprite Scale=DataInterface() {propertyValues {Curve {keys [
        {interpMode=RCIM_Cubic, tangentMode=RCTM_User, time=0, value=0}
        <at 0.2 1 RCIM_Cubic>
        {arriveTangent=-2.776, interpMode=RCIM_Cubic, leaveTangent=-2.776, tangentMode=RCTM_User, time=1, value=0}
       ]}}})
   &CurlNoiseForce(Noise Frequency=3.2204, Noise Strength=300, Pan Noise=false)
   &ScaleVelocity(
    Coordinate Space:ENiagaraCoordinateSpace=::World
    Velocity Scale=&MakeVector(
     X=1
     Y=1
     Z=&FloatFromCurve(FloatCurve=DataInterface() {propertyValues {Curve {keys [<at 0 1 RCIM_Cubic>, {arriveTangent=1.2298, interpMode=RCIM_Cubic, leaveTangent=1.2298, tangentMode=RCTM_User, time=1, value=1.4}]}}})
    )
   )
   &SolveForcesAndVelocity(Clamp Velocity=false, Limit Acceleration=false)
  ]
  rendererClasses [&NiagaraSpriteRendererProperties]
  renderers [{rendererClass=&NiagaraSpriteRendererProperties}]
 }
 Leaves_02 {
  emitterUpdateScript [
   &EmitterState()
   &SpawnRate(SpawnRate=15, Use Spawn Probability=false)
  ]
  particleSpawnScript [
   &InitializeParticle(
    Color=&SelectLinearColorFromArray(
     Array Sampling Mode=::NewEnumerator0
     Direct Array Index=?
     Color Selection Array=DataInterface() {propertyValues {ColorData [<rgba 0.4427 0.0503 0.0679 1>, <rgba 0.9911 0.6038 0.0999 1>, <rgba 0.3489 0.3763 0.0746 1>, <rgba 1 0.3419 0 1>]}}
     Random Seed=0
    )
    Lifetime Max=1
    Lifetime Min=0.2
    Lifetime Mode=::NewEnumerator1
    Mesh Scale Mode=::NewEnumerator3
    Sprite Rotation Mode=::NewEnumerator2
    Sprite Size Max=<15 15>
    Sprite Size Min=<7.5 7.5>
    Sprite Size Mode=::NewEnumerator2
    UsePositionOffset=false
    Write Material Random=true
   )
   &SystemLocation(MaskSpawnGroup=false, Offset=<0 0 80>)
   &AddVelocity(
    Cone Angle=70
    Cone Axis=<0 1 1>
    Module.UseConeFalloff=false
    Module.UseVelDistribution=false
    Override Seed=false
    Velocity Mode=::NewEnumerator2
    Velocity Speed=&RandomRangeFloat(Minimum=300, Maximum=600)
   ) {module=AddVelocity001}
  ]
  particleUpdateScript [
   &ParticleState()
   &SubUVAnimation(
    Play Rate:NiagaraFloat=&RandomRangeFloat(Minimum=1, Maximum=3)
    Playback Scale Curve=DataInterface() {propertyValues {Curve {keys [{interpMode=RCIM_Cubic, tangentMode=RCTM_User, time=0, value=0}, {interpMode=RCIM_Cubic, tangentMode=RCTM_User, time=1, value=1}]}}}
    "Sprite Renderer "=DataInterface() {propertyValues {SpriteRenderer=&self:Leaves_02_4.NiagaraSpriteRendererProperties_1}}
    SubUV Animation Mode:ENiagara_SubUVLookupModeV2=::NewEnumerator3
    UseEndFrame=true
    UsePlayRateScale=true
    UseStartFrame=true
   ) {module=SubUVAnimation002}
   &Drag(
    Drag=&FloatFromCurve(
     FloatCurve=DataInterface() {propertyValues {Curve {keys [
         {arriveTangent=3.7695, interpMode=RCIM_Cubic, leaveTangent=3.7695, tangentMode=RCTM_User, time=0, value=0}
         {interpMode=RCIM_Cubic, tangentMode=RCTM_User, time=1, value=1}
        ]}}}
     Scale Curve=2
    )
    Use Linear Drag=true
    Use Rotational Drag=true
   )
   &DynamicMaterialParameters(
    Index 0 Param 1=0.1
    Module.Param0WriteEnabled=true
    Module.Param1WriteEnabled=true
    Module.Param2WriteEnabled=true
    Module.Param3WriteEnabled=true
   )
   &ScaleColor(
    Scale Alpha=&FloatFromCurve(FloatCurve=DataInterface() {propertyValues {Curve {keys [
         {interpMode=RCIM_Cubic, tangentMode=RCTM_User, time=0, value=0}
         {arriveTangent=-0.0948, interpMode=RCIM_Cubic, leaveTangent=-0.0948, tangentMode=RCTM_User, time=0.1, value=1}
         <at 0.942 0.9825>
         <at 1 0>
        ]}}})
    Scale RGB=&MakeVectorFromLinearColorRGB(Color=&HueShiftLinearColor(
      Hue=&RandomRangeFloat(Minimum=0, Maximum=0.05)
      Saturation=1
      Value=&Multiply_Float(
       A=&RandomRangeFloat(Minimum=1.3, Maximum=4)
       B=&FNTimeOfDayScalar(Night Scale=0.5, Morning Scale=1, Evening Scale=0.8)
      )
      Linear Color=$Particles.Initial.Color
     ))
    ScaleA=false
    ScaleRGB=true
   )
   &ScaleSpriteSize(Uniform Curve Sprite Scale=DataInterface() {propertyValues {Curve {keys [
        {interpMode=RCIM_Cubic, tangentMode=RCTM_User, time=0, value=0}
        <at 0.2 1 RCIM_Cubic>
        {arriveTangent=-2.776, interpMode=RCIM_Cubic, leaveTangent=-2.776, tangentMode=RCTM_User, time=1, value=0}
       ]}}})
   &SolveForcesAndVelocity(Clamp Velocity=false, Limit Acceleration=false)
  ]
  rendererClasses [&NiagaraSpriteRendererProperties]
  renderers [{rendererClass=&NiagaraSpriteRendererProperties}]
 }
 Leaves_Collision {
  emitterUpdateScript [
   &EmitterState()
   &SpawnRate(SpawnRate=5, Use Spawn Probability=false)
  ]
  particleSpawnScript [
   &InitializeParticle(
    Color=&SelectLinearColorFromArray(
     Array Sampling Mode=::NewEnumerator0
     Direct Array Index=?
     Color Selection Array=DataInterface() {propertyValues {ColorData [<rgba 0.4427 0.0503 0.0679 1>, <rgba 0.9911 0.6038 0.0999 1>, <rgba 0.3489 0.3763 0.0746 1>, <rgba 1 0.3419 0 1>]}}
     Random Seed=0
    )
    Lifetime Max=4
    Lifetime Min=3
    Lifetime Mode=::NewEnumerator1
    Mesh Scale Mode=::NewEnumerator3
    Sprite Rotation Mode=::NewEnumerator2
    Sprite Size Max=<15 15>
    Sprite Size Min=<7.5 7.5>
    Sprite Size Mode=::NewEnumerator2
    UsePositionOffset=false
    Write Material Random=true
   )
   &SystemLocation(MaskSpawnGroup=false, Offset=<0 0 80>)
   &AddVelocity(
    Cone Angle=70
    Cone Axis=<0 0.5 1>
    Module.UseConeFalloff=false
    Module.UseVelDistribution=false
    Override Seed=false
    Velocity Mode=::NewEnumerator2
    Velocity Speed=&RandomRangeFloat(Minimum=150, Maximum=300)
   ) {module=AddVelocity001}
  ]
  particleUpdateScript [
   &ParticleState()
   &Drag(
    Drag=&FloatFromCurve(
     FloatCurve=DataInterface() {propertyValues {Curve {keys [
         {arriveTangent=3.7695, interpMode=RCIM_Cubic, leaveTangent=3.7695, tangentMode=RCTM_User, time=0, value=0}
         {interpMode=RCIM_Cubic, tangentMode=RCTM_User, time=1, value=1}
        ]}}}
     Scale Curve=2
    )
    Use Linear Drag=true
    Use Rotational Drag=true
   )
   &DynamicMaterialParameters(
    Index 0 Param 1=0.1
    Module.Param0WriteEnabled=true
    Module.Param1WriteEnabled=true
    Module.Param2WriteEnabled=true
    Module.Param3WriteEnabled=true
   )
   &ScaleColor(
    Scale Alpha=&FloatFromCurve(FloatCurve=DataInterface() {propertyValues {Curve {keys [
         {interpMode=RCIM_Cubic, tangentMode=RCTM_User, time=0, value=0}
         {arriveTangent=-0.0948, interpMode=RCIM_Cubic, leaveTangent=-0.0948, tangentMode=RCTM_User, time=0.1, value=1}
         <at 0.942 0.9825>
         <at 1 0>
        ]}}})
    Scale RGB=&MakeVectorFromLinearColorRGB(Color=&HueShiftLinearColor(
      Hue=&RandomRangeFloat(Minimum=0, Maximum=0.05)
      Saturation=1
      Value=&Multiply_Float(
       A=&RandomRangeFloat(Minimum=1.3, Maximum=4)
       B=&FNTimeOfDayScalar(Night Scale=0.5, Morning Scale=1, Evening Scale=0.8)
      )
      Linear Color=$Particles.Initial.Color
     ))
    ScaleA=false
    ScaleRGB=true
   )
   &ScaleSpriteSize(Uniform Curve Sprite Scale=DataInterface() {propertyValues {Curve {keys [
        {interpMode=RCIM_Cubic, tangentMode=RCTM_User, time=0, value=0}
        {arriveTangent=1.25, leaveTangent=1.25, time=0.1, value=1}
        {arriveTangent=-1.25, leaveTangent=-1.25, time=0.9, value=1}
        {arriveTangent=-2.776, interpMode=RCIM_Cubic, leaveTangent=-2.776, tangentMode=RCTM_User, time=1, value=0}
       ]}}})
   &CurlNoiseForce(Noise Frequency=3.2204, Noise Strength=40, Pan Noise=false)
   &GravityForce(Gravity=<0 0 -100>)
   &Collision(
    Advanced Aging Rate=1.3
    Age Colliding Particles=true
    Friction=5
    KillOnCollision=false
    Module.Control Rest State Via Surface Normal=false
    Module.EnableMaxCollisionCount=false
    Module.Randomize Collision Normal Vector=false
    Particle Radius Scale=0.8
    Restitution=0
   )
   &SubUVAnimation(
    Play Rate:NiagaraFloat=&MakeCustomFloatFromBool(Bool=$Particles.HasCollided, True Float=0, False Float=1)
    Random Start Frame=true
    "Sprite Renderer "=DataInterface() {propertyValues {SpriteRenderer=&self:Leaves_Collision_0.NiagaraSpriteRendererProperties_1}}
    SubUV Animation Mode:ENiagara_SubUVLookupModeV2=::NewEnumerator3
    UseEndFrame=true
    UsePlayRateScale=true
    UseStartFrame=true
   ) {module=SubUVAnimation002}
   &SolveForcesAndVelocity(Clamp Velocity=false, Limit Acceleration=false)
  ]
  rendererClasses [&NiagaraSpriteRendererProperties]
  renderers [{rendererClass=&NiagaraSpriteRendererProperties}]
 }
}
summary {
 systemName=NS_Emote_LeafBlower_Leaves_Loop
 emitters [
  {emitterName=Leaves_01, rendererClasses [&NiagaraSpriteRendererProperties]}
  {emitterName=Leaves_02, rendererClasses [&NiagaraSpriteRendererProperties]}
  {emitterName=Leaves_Collision, rendererClasses [&NiagaraSpriteRendererProperties]}
 ]
 userVariables [User.IsFrontEnd {type=NiagaraBool}]
}
