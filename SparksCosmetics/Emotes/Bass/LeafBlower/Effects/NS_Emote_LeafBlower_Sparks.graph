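# Niagara system for a short burst of orange spark sprites, used by the LeafBlower bass emote in SparksCosmetics.
# A single Sparks emitter burst-spawns small, short-lived HDR-colored particles and launches them in a wide cone at randomized speeds.
# Drag and randomized downward gravity shape their motion, while sprite size scales with speed and shrinks away late in life.
# An AddVelocityFromPoint module keyed to spawn group is present but disabled.
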NiagaraSystem /SparksCosmetics/Emotes/Bass/LeafBlower/Effects/NS_Emote_LeafBlower_Sparks
dependencies {usedDataInterfaces [&NiagaraDataInterfaceCurve], usedDynamicInputs [], usedModules [&SystemState]}
emitters {Sparks {
  emitterUpdateScript [
   &EmitterState()
   &SpawnBurst_Instantaneous(Spawn Count=10, Use Spawn Probability=false, UseLoopCountLimit=false)
  ]
  particleSpawnScript [
   &InitializeParticle(
    Color=<rgba 10 3.0725 0 1>
    Lifetime=&RandomRangeFloat(Minimum=0.1, Maximum=0.3)
    Sprite Size Max=<1.5 4>
    Sprite Size Min=<0.5 2>
    Sprite Size Mode=::NewEnumerator2
    UsePositionOffset=false
    Write Material Random=true
   )
   &AddVelocityFromPoint(
    Module.Use Falloff Distance=false
    Velocity Strength=&MakeCustomFloatFromBool(
     Bool=HlslExpression() {hlslExpression="System.SpawnGroup > 1"}
     True Float=&RandomRangeFloat(Minimum=100, Maximum=400)
     False Float=0
    )
   ) {enabled=false}
   &AddVelocityInCone(
    Cone Angle=125
    Cone Axis=<1 0 1>
    Module.Use Velocity Falloff On Cone Axis=false
    Velocity Strength=&RandomRangeFloat(Minimum=100, Maximum=300)
   )
  ]
  particleUpdateScript [
   &ParticleState()
   &Drag(Use Linear Drag=true, Use Rotational Drag=false)
   &GravityForce(Gravity=&MakeVector(X=0, Y=0, Z=&RandomRangeFloat(Minimum=-980, Maximum=-450)))
   &SolveForcesAndVelocity(Clamp Velocity=false, Limit Acceleration=false)
   &ScaleSpriteSizeBySpeed(Max Scale Factor=<0.5 2>, Velocity Threshold=120)
   &ScaleSpriteSize(Scale Factor=&Vector2DFromFloat(Value=&FloatFromCurve(FloatCurve=DataInterface() {propertyValues {Curve {keys [{interpMode=RCIM_Cubic, tangentMode=RCTM_User, time=0.7, value=1}, {interpMode=RCIM_Cubic, tangentMode=RCTM_User, time=1, value=0}]}}})))
  ]
  rendererClasses [&NiagaraSpriteRendererProperties]
  renderers [{rendererClass=&NiagaraSpriteRendererProperties}]
 }}
summary {
 systemName=NS_Emote_LeafBlower_Sparks
 emitters [{emitterName=Sparks, rendererClasses [&NiagaraSpriteRendererProperties]}]
 userVariables []
}
